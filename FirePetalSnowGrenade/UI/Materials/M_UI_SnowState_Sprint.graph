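# Translucent UI material for a progress bar, apparently for the sprint snow state.
# Fill combines a blue color blend with named background noise, rotated crack pattern and reflected-gradient glow layers.
# Progress drives a hard horizontal cutoff between backing and fill alphas, plus a rotated, angled marker edge with animatable tint.
# Static switches SimpleProgress and HasBacking select plain fill over backing, marked fill over backing, or marked fill alone.
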Material /FirePetalSnowGrenade/UI/Materials/M_UI_SnowState_Sprint
graph {
 properties {blendMode=BLEND_Translucent, materialDomain=MD_UI, pixelDepthOffsetMode=PDOM_Legacy, usedShadingModels=Default Lit}
 nodes [
  n1 = MaterialFunctionCall() {materialFunction=&GetUserInterfaceUV}
  n2 = Add(A=Subtract(A=OneMinus(None=@n1.Normalized UV)), B=ScalarParameter() {defaultValue=1, group=Bar, parameterName=Progress})
  n3 = Saturate(None=Multiply(A=ComponentMask(None=@n2) {r=true}, B=ScalarParameter() {defaultValue=1024, group=Bar, parameterName=Progress_Edge_Hardness}))
  LinearInterpolate(
   A=ScalarParameter() {defaultValue=0.5, group=Bar, parameterName=Progress_Backing_Alpha}
   B=ScalarParameter() {defaultValue=1, group=Bar, parameterName=Progress_Fill_Alpha}
   Alpha=@n3
  )
  n4 = ScalarParameter() {defaultValue=1, parameterName=BgTilingX}
  n5 = TextureCoordinate()
  n6 = ComponentMask(None=@n5) {r=true}
  n7 = ComponentMask(None=@n5) {g=true}
  n8 = ScalarParameter() {defaultValue=1, parameterName=BgTillingY}
  n9 = TextureSample(UVs=AppendVector(A=Multiply(A=@n4, B=@n6), B=Multiply(A=@n7, B=@n8))) {
   samplerType=SAMPLERTYPE_LinearGrayscale
   texture=/Game/Characters/Player/Female/Medium/Bodies/F_MED_SuperNovaTaro/Textures/LowResBlurredNoise
  }
  NamedRerouteDeclaration(None=@n9.RGB) {name=Bg, nodeColor=<rgba 0 0 0 1>}
  n10 = ScalarParameter() {defaultValue=10, parameterName=TilingX}
  n11 = TextureCoordinate()
  n12 = ComponentMask(None=@n11) {r=true}
  n13 = ComponentMask(None=@n11) {g=true}
  n14 = ScalarParameter() {defaultValue=0.5, parameterName=TilingY}
  n15 = MaterialFunctionCall(UVs=AppendVector(A=Multiply(A=@n10, B=@n12), B=Multiply(A=@n13, B=@n14)), "Rotation Angle (0-1)"=Constant() {r=0.1}) {materialFunction=&CustomRotator}
  NamedRerouteDeclaration(None=Multiply(
    A=TextureSample(UVs=@n15.Rotated Values) {texture=&/T_UI_Noise_Cracks}.RGB
    B=ScalarParameter() {defaultValue=1, parameterName=PatternIntesity}
   )) {name=Pattern, nodeColor=<rgba 0 0 0 1>}
  n16 = OneMinus(None=MaterialFunctionCall!(UVs, Offset, RotationCenter, RotationAngle=Constant() {r=0.25}) {materialFunction=/FortUILibrary/MaterialFunctions/Gradients/MF_UI_ReflectedGradient}.Result)
  n17 = ScalarParameter() {defaultValue=3, parameterName=GlowIntensity}
  NamedRerouteDeclaration(None=Multiply(A=Power(Base=@n16, Exp=@n17), B=ScalarParameter() {defaultValue=0.5, parameterName=GlowOpacity}) {constB=0.5}) {name=Glow, nodeColor=<rgba 0 0 0 1>}
  n18 = LinearInterpolate(
   A=VectorParameter() {group=Color, parameterName=Backing_Color, defaultValue=<rgba 0 0 0 1>}.RGB
   B=VectorParameter() {group=Color, parameterName=Backing_Color_Stop, defaultValue=<rgba 0 0 0 1>}.RGB
   Alpha=ScalarParameter() {parameterName=AnimateBackingColor}
  )
  n19 = VectorParameter() {group=Color, parameterName=BackgroundColor, defaultValue=<rgba 0 0.0742 0.9844 1>}
  n20 = VectorParameter() {group=Color, parameterName=BrightColor, defaultValue=<rgba 0.2891 0.3963 1 1>}
  n21 = Ceil(None=ComponentMask(None=@n1.Normalized UV) {r=true})
  n22 = Saturate(None=Add(A=OneMinus(None=@n21), B=NamedRerouteUsage()))
  n23 = Add(A=Add(A=LinearInterpolate(A=@n19.RGB, B=@n20.RGB, Alpha=@n22), B=NamedRerouteUsage()), B=NamedRerouteUsage())
  n24 = VectorParameter() {group=Color, parameterName=Marker_Color, defaultValue=<rgba 0 0.5378 1 1>}
  n25 = ScalarParameter() {parameterName=LiftBlendFadeStart}
  n26 = ComponentMask(None=Add(A=TextureCoordinate(), B=ScalarParameter() {parameterName=OffsetBlendFadeStart})) {r=true}
  n27 = Saturate(None=DotProduct(A=LinearInterpolate(A=@n25, Alpha=@n26), B=ScalarParameter() {group=Marker, parameterName=BlendMarkerColor}))
  n28 = ScalarParameter() {defaultValue=-0.0125, group=Marker, parameterName=MarkerAngle_Rotation}
  n29 = Subtract(
   A=OneMinus(None=MaterialFunctionCall(UVs=@n1.Normalized UV, "Rotation Angle (0-1)"=@n28) {materialFunction=&CustomRotator}.Rotated Values)
   B=Add(A=Constant() {r=1}, B=ScalarParameter() {defaultValue=0.0625, group=Marker, parameterName=MarkerAngle_Offset})
  )
  n30 = ScalarParameter() {defaultValue=1, group=Bar, parameterName=Progress}
  n31 = Multiply(A=ComponentMask(None=Add(A=@n29, B=@n30)) {r=true}, B=ScalarParameter() {defaultValue=80, group=Marker, parameterName=Angle_Edge_Hardness})
  n32 = LinearInterpolate(
   A=@n23
   B=LinearInterpolate(A=LinearInterpolate(A=@n23, B=@n24.RGB, Alpha=@n27), B=@n24.RGB, Alpha=ScalarParameter() {parameterName=AnimateMarkerIntensity})
   Alpha=OneMinus(None=Saturate(None=@n31))
  )
  StaticSwitchParameter(
   True=LinearInterpolate(A=@n18, B=@n23, Alpha=@n3)
   False=StaticSwitchParameter(True=LinearInterpolate(A=@n18, B=@n32, Alpha=@n3), False=@n32) {parameterName=HasBacking}
  ) {parameterName=SimpleProgress}
 ]
}
