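# Translucent UI material for a snow-themed glow in FirePetalSnowGrenade, built around a progress bar with separate backing and fill color and alpha.
# A hardened, saturated UV threshold driven by Progress blends backing into fill, which uses an overlay-blended Color.
# The "Nitro" reroute sums an MF_UI_Box mask and a second one modulated by two multiplied panning noise samples, one rotated.
# A separate pair of panning swirly noise textures is multiplied, powered, boosted and saturated.
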Material /FirePetalSnowGrenade/UI/Materials/M_UI_Snow_Glow
graph {
 properties {blendMode=BLEND_Translucent, materialDomain=MD_UI, pixelDepthOffsetMode=PDOM_Legacy, usedShadingModels=Default Lit}
 nodes [
  n1 = MaterialFunctionCall() {materialFunction=&GetUserInterfaceUV}
  n2 = Add(A=Subtract(A=OneMinus(None=@n1.Normalized UV)), B=ScalarParameter() {defaultValue=0.5, group=Bar, parameterName=Progress})
  n3 = Saturate(None=Multiply(A=ComponentMask(None=@n2) {r=true}, B=ScalarParameter() {defaultValue=1024, group=Bar, parameterName=Progress_Edge_Hardness}))
  LinearInterpolate(
   A=LinearInterpolate(
    A=VectorParameter() {group=Color, parameterName=Backing_Color, defaultValue=<rgba 0 0 0 1>}.RGB
    B=VectorParameter() {group=Color, parameterName=Backing_Color_Stop, defaultValue=<rgba 0 0 0 1>}.RGB
    Alpha=ScalarParameter() {parameterName=AnimateBackingColor}
   )
   B=MaterialFunctionCall(Base=VectorParameter() {parameterName=Color, defaultValue=<rgba 0 0.219 1 1>}.RGB, Blend=NamedRerouteUsage()) {materialFunction=&Blend_Overlay}.Result
   Alpha=@n3
  )
  Multiply(
   A=LinearInterpolate(
    A=ScalarParameter() {defaultValue=0.5, group=Bar, parameterName=Progress_Backing_Alpha}
    B=ScalarParameter() {defaultValue=1, group=Bar, parameterName=Progress_Fill_Alpha}
    Alpha=@n3
   )
   B=NamedRerouteUsage()
  )
  n4 = TextureCoordinate() {unMirrorU=true}
  n5 = Constant()
  n6 = Constant() {r=-0.4}
  n7 = Constant()
  n8 = Constant() {r=0.9}
  n9 = MaterialFunctionCall(UVs=@n4, Size=Constant() {r=0.2}, CornerRoundness=@n6, GlowMin=Constant(), GlowMax=Constant() {r=0.8}, KeepAspectRatio=StaticBool()) {materialFunction=&MF_UI_Box}
  n10 = TextureCoordinate()
  n11 = AppendVector(
   A=Multiply(A=ScalarParameter() {defaultValue=1, parameterName=TilingX}, B=ComponentMask(None=@n10) {r=true})
   B=Multiply(A=ComponentMask(None=@n10) {g=true}, B=ScalarParameter() {defaultValue=1, parameterName=TilingY})
  )
  n12 = Time()
  n13 = ScalarParameter() {defaultValue=0.5, parameterName=Speed1}
  n14 = TextureSample(UVs=Panner(
    Coordinate=@n11
    Time=Multiply(A=@n12, B=@n13)
    Speed=VectorParameter() {parameterName=PannerDirection1, defaultValue=<rgba 0.2 0.3 0 0>}.RGB
   ) {speedX=0.2, speedY=0.6}) {
   samplerType=SAMPLERTYPE_LinearGrayscale
   texture=/Game/Characters/Player/Female/Medium/Bodies/F_MED_SuperNovaTaro/Textures/LowResBlurredNoise
  }
  n15 = Constant() {r=0.1}
  n16 = Time()
  n17 = ScalarParameter() {defaultValue=0.3, parameterName=Speed2}
  n18 = TextureSample(UVs=Panner(
    Coordinate=MaterialFunctionCall(UVs=@n11, "Rotation Angle (0-1)"=@n15) {materialFunction=&CustomRotator}.Rotated Values
    Time=Multiply(A=@n16, B=@n17)
    Speed=VectorParameter() {parameterName=PannerDirection2, defaultValue=<rgba -0.4 0.2 0 0>}.RGB
   ) {speedX=0.2, speedY=0.6}) {
   samplerType=SAMPLERTYPE_LinearGrayscale
   texture=&LowResBlurredNoise
  }
  n19 = Multiply(A=@n14.RGB, B=@n18.RGB)
  NamedRerouteDeclaration(None=Add(
    A=MaterialFunctionCall(UVs=@n4, Size=@n5, CornerRoundness=@n6, GlowMin=@n7, GlowMax=@n8) {materialFunction=&MF_UI_Box}.Fill
    B=Multiply(A=@n9.Fill, B=@n19)
   )) {name=Nitro, nodeColor=<rgba 0 0 0 1>}
  n20 = TextureCoordinate() {uTiling=2, vTiling=0.05}
  n21 = Multiply(A=Time(), B=Constant() {r=1})
  n22 = VectorParameter() {parameterName=Speed1, defaultValue=<rgba 0.2511 0.2 0 1>}
  n23 = TextureCoordinate() {uTiling=3, vTiling=0.1}
  n24 = Multiply(A=Time(), B=Constant() {r=0.6})
  n25 = VectorParameter() {parameterName=Speed2, defaultValue=<rgba 0.2125 0.2 0 1>}
  n26 = Multiply(
   A=TextureSampleParameter2D(UVs=Panner(Coordinate=@n20, Time=@n21, Speed=@n22.RGB) {speedX=0.3, speedY=0.8}) {parameterName=Texture01, texture=&/T_Swirly_Noise_LowDetail}.RGB
   B=TextureSampleParameter2D(UVs=Panner(Coordinate=@n23, Time=@n24, Speed=@n25.RGB) {speedX=0.2, speedY=0.6}) {parameterName=Texture02, texture=&T_Swirly_Noise_LowDetail}.RGB
  )
  n27 = ScalarParameter() {defaultValue=1, parameterName=Power_Texture}
  Saturate(None=Multiply(A=Power(Base=@n26, Exp=@n27) {constExponent=1.8}, B=ScalarParameter() {defaultValue=8, parameterName=Multiply_Texture}))
  StaticBool() {value=true}
 ]
}
