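# UI master material that draws a two-tone icon with fill and shadow, named for TacSprint but sampling the snow grenade icon texture.
# The texture's alpha serves as a distance field, thresholded twice through MF_UI_AntiAliasing and inverted to form fill and shadow masks.
# The fill mask blends shadow color into fill color, and a separate sum adds it to the opacity-scaled shadow mask.
# Colors, both thresholds and shadow opacity are parameters.
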Material /FirePetalSnowGrenade/UI/Materials/M_UI_TacSprintIcon_Master
graph {
 properties {blendMode=BLEND_Translucent, materialDomain=MD_UI, pixelDepthOffsetMode=PDOM_Legacy, usedShadingModels=Default Lit}
 nodes [
  n1 = TextureSample(UVs=MaterialFunctionCall(UVs=TextureCoordinate(), Texture Scale=Constant() {r=1}) {materialFunction=&ScaleUVsByCenter}.UVs) {samplerType=SAMPLERTYPE_LinearGrayscale, texture=/FirePetalSnowGrenade/UI/Textures/T_UI_Icon_SnowGrenade}
  n2 = OneMinus(None=MaterialFunctionCall(DistanceField=@n1.A, Threshold=ScalarParameter() {defaultValue=0.75, parameterName=FillSmoothness}) {materialFunction=&MF_UI_AntiAliasing}.Inner)
  n3 = MaterialFunctionCall(DistanceField=@n1.A, Threshold=ScalarParameter() {defaultValue=0.6, parameterName=ShadowSmoothness}) {materialFunction=&MF_UI_AntiAliasing}
  Add(A=@n2, B=Multiply(A=OneMinus(None=@n3.Inner), B=ScalarParameter() {defaultValue=0.5, parameterName=ShadowOpacity}) {constB=0.5})
  LinearInterpolate(
   A=VectorParameter() {parameterName=ShadowColor}.RGB
   B=VectorParameter() {group=Color, parameterName=FillColor, defaultValue=<rgba 1 1 1 1>}.RGB
   Alpha=@n2
  )
 ]
}
