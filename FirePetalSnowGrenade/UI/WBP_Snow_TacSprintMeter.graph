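# Tactical sprint meter child widget that shows a snow-themed stamina bar while the owning player has the snow grenade slow tag.
# Async tag waits toggle the overlay and sprint icon.
# The SetStaminaProgress override drives fill and glow dynamic materials, dims the icon when empty, and briefly shows large energy drops on a timer-cleared delta bar.
# Blink logic for the effect's final seconds exists, but its polling timer and handler call are unwired.
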WidgetBlueprint /FirePetalSnowGrenade/UI/WBP_Snow_TacSprintMeter
graphs [
 CacheMaterials {nodes [
   CacheMaterials() {is=title K2Node_FunctionEntry}
   > n1 = GetDynamicMaterial(self=Get Snow_DeltaBar() -> (Snow_DeltaBar:Image Object Reference))
   > Set DeltaBarMID(DeltaBarMID:Material Instance Dynamic Object Reference=@n1) -> (Output_Get:Material Instance Dynamic Object Reference)
   > n2 = GetDynamicMaterial(self=Get Snow_FillBar() -> (Snow_FillBar:Image Object Reference))
   > Set FillBarMID(FillBarMID:Material Instance Dynamic Object Reference=@n2) -> (Output_Get:Material Instance Dynamic Object Reference)
   > n3 = GetDynamicMaterial(self=Get Snow_OuterGlow() -> (Snow_OuterGlow:Image Object Reference))
   > Set GlowBarMID(GlowBarMID:Material Instance Dynamic Object Reference=@n3) -> (Output_Get:Material Instance Dynamic Object Reference)
   Get FillBar() -> (FillBar:Image Object Reference)
   Get DeltaBar() -> (DeltaBar:Image Object Reference)
  ]}
 CalculateHideInterval {nodes [
   n1 = Inputs!() -> (execute, RemainingTime:double=~)
   n2 = Truncate(A=float * float(A=@n1, B=10) {is=title K2Node_PromotableOperator})
   > Outputs!(then:Exec, bShouldHide:bool="float < float"(A:int="% (Integer)"(A=@n2, B=6) -> (ReturnValue=0), B=3))
  ]}
 EmptyDelta {nodes [
   EmptyDelta() {is=title K2Node_FunctionEntry}
   > SetScalarParameterValue(self=Get DeltaBarMID() -> (DeltaBarMID:Material Instance Dynamic Object Reference), ParameterName=Progress, Value=0)
  ]}
 EventGraph {nodes [
   n1 = OnPeriodicDurationUpdate() {is=title K2Node_CustomEvent}
   > n2 = FindLongestActiveEffectDuration!(execute, self) -> (then, TimeRemaining:double)
   WaitSnowTag() {is=title K2Node_CustomEvent}
   n3 = GetOwningPlayer()
   > Branch(Condition=AND Boolean(A=IsUsingTouch(PlayerController=@n3), B=IsMobileGame()))
   n4 = GetOwningPlayer()
   > n5 = WaitGameplayTagAddToActor!(
    execute:Exec
    TargetActor:Actor Object Reference=@n4
    Tag:Gameplay Tag Structure=Get SnowTag_Active() -> (SnowTag_Active:Gameplay Tag Structure)
    OnlyTriggerOnce:bool=false
   ) -> (then:Exec, AsyncTaskProxy:Ability Async Wait Gameplay Tag Added Object Reference, Added:Exec) {is=title K2Node_AsyncAction}
   > n6 = WaitGameplayTagRemoveFromActor!(execute:Exec, TargetActor:Actor Object Reference=@n4, Tag:Gameplay Tag Structure=Get SnowTag_Active(), OnlyTriggerOnce:bool=false) -> (then:Exec, AsyncTaskProxy:Ability Async Wait Gameplay Tag Removed Object Reference, Removed:Exec) {is=title K2Node_AsyncAction}
   SetVisibility(
    execute=@n6.Removed
    self=[Get SprintIcon() -> (SprintIcon:Image Object Reference), Get SnowSprintOverlay() -> (SnowSprintOverlay:Overlay Object Reference)]
    InVisibility=Collapsed
   )
   SetVisibility(execute=@n5.Added, self=[Get SprintIcon(), Get SnowSprintOverlay()], InVisibility=SelfHitTestInvisible)
   Event Construct()
   > Parent: Construct()
   n7 = GetOwningPlayer()
   > Branch(Condition=AND Boolean(A=IsUsingTouch(PlayerController=@n7), B=IsMobileGame()))
   > SetPadding!(
    execute
    self:Overlay Slot Object Reference=SlotAsOverlaySlot(Widget=Get HorizontalBox_Energy() -> (HorizontalBox_Energy:Horizontal Box Object Reference))
    InPadding_Left:float=-28
    InPadding_Top:float=-24
    InPadding_Right:float=0
    InPadding_Bottom:float=0
   ) -> (then)
   Event OnInitialized()
   > CacheMaterials() {is=title K2Node_CallFunction}
   n8 = IsUsingTouch(PlayerController=GetOwningPlayer())
   n9 = IsMobileGame()
   > SetRenderTranslation(Translation=Select(Option 0:Vector 2D Structure, Option 1:Vector 2D Structure=(X=1,Y=0), Index=AND Boolean(A=@n8, B=@n9)) -> (ReturnValue:Vector 2D Structure))
   > WaitSnowTag() {is=title K2Node_CallFunction}
   Set Timer by Event(Delegate=@n1, Time=0.1, bLooping=true, bMaxOncePerFrame=true)
   HandleSnowEffectEndingState!(execute, self, RemainingTime:double=@n2) -> (then)
  ]}
 FindLongestActiveEffectDuration {nodes [
   FindLongestActiveEffectDuration() {is=title K2Node_FunctionEntry}
   > n1 = GetActiveEffectsWithAllTags!(
    execute
    self:Ability System Component Object Reference=GetAbilitySystemComponent!(Actor:Actor Object Reference=GetOwningPlayer()) -> (ReturnValue:Ability System Component Object Reference)
    Tags:Gameplay Tag Container Structure=MakeGameplayTagContainerFromTag!(SingleTag:Gameplay Tag Structure=Get ConstantSnowEffectTag() -> (ConstantSnowEffectTag:Gameplay Tag Structure)) -> (ReturnValue:Gameplay Tag Container Structure)
   ) -> (then, ReturnValue:Array of Active Gameplay Effect Handle Structures)
   > n2 = For Each Loop(Array:Array of Active Gameplay Effect Handle Structures=@n1) -> (Array Element:Active Gameplay Effect Handle Structure)
   Return Node!(execute=@n2.Completed, TimeRemaining:double=Get LongestTimeRemaining() -> (LongestTimeRemaining:double))
   > n3 = GetActiveGameplayEffectRemainingDuration!(execute, ActiveHandle:Active Gameplay Effect Handle Structure=@n2) -> (then, ReturnValue:float)
   > Branch(Condition="float > float"(A=@n3, B=Get LongestTimeRemaining()))
   > Set LongestTimeRemaining(LongestTimeRemaining:double=@n3) -> (Output_Get:double)
  ]}
 HandleSnowEffectEndingState {nodes [
   n1 = HandleSnowEffectEndingState!() -> (then, RemainingTime:double=~) {is=title K2Node_FunctionEntry}
   n2 = Get ConstantEndingStateTime() -> (ConstantEndingStateTime:double)
   n3 = "float <= float"(A=@n1, B=@n2) -> (ReturnValue=~)
   n4 = Get bEndingStateActive() -> (bEndingStateActive:bool)
   > n5 = Branch(Condition=AND Boolean(A="float > float"(A=@n2), B="Not Equal (Boolean)"(A=@n3, B=@n4) -> (ReturnValue=~)))
   > n6 = Set bEndingStateActive(bEndingStateActive:bool=@n3) -> (Output_Get:bool)
   n7 = Branch(execute=[@n6.then, @n5.else], Condition=@n6)
   > n8 = Calculate Hide Interval!(execute:Exec, RemainingTime:double=Get RemainingTime() -> (RemainingTime:double)) -> (then:Exec, bShouldHide:bool=~) {is=title K2Node_MacroInstance}
   > n9 = Set bShouldHide(bShouldHide:bool=@n8) -> (Output_Get:bool)
   Branch(execute=[@n9.then, @n10.then], Condition="Not Equal (Boolean)"(A=@n9, B=Get bAnimHideWidget() -> (bAnimHideWidget:bool)) -> (ReturnValue=~))
   > Set bAnimHideWidget(bAnimHideWidget:bool=Get bShouldHide() -> (bShouldHide:bool)) -> (Output_Get:bool)
   > SetVisibility(
    self=Get SnowSprintOverlay()
    InVisibility=Select(Option 0:ESlateVisibility Enum=SelfHitTestInvisible, Option 1:ESlateVisibility Enum=Hidden, Index=Get bAnimHideWidget()) -> (ReturnValue:ESlateVisibility Enum)
   )
   n10 = Set bShouldHide(execute=@n7.else, bShouldHide=false)
  ]}
 SetDelta {nodes [
   n1 = SetDelta!() -> (then, CurrentEnergy:double=~) {is=title K2Node_FunctionEntry}
   n2 = Get Anim_DeltaThreshold() -> (Anim_DeltaThreshold:double)
   n3 = Get PrevEnergy() -> (PrevEnergy:double)
   > n4 = Branch(Condition="float <= float"(A=@n1, B=float - float(A=@n3, B=@n2)) -> (ReturnValue=~))
   > SetScalarParameterValue(self=Get DeltaBarMID(), ParameterName=Progress, Value=Get PrevEnergy())
   > Clear and Invalidate Timer by Handle(Handle=Get EmptyDelta_Handle() -> (EmptyDelta_Handle:Timer Handle Structure))
   > n5 = Set Timer by Event(Delegate=Create Event(), Time=0.33, bLooping=false)
   > Set EmptyDelta_Handle(EmptyDelta_Handle:Timer Handle Structure=@n5) -> (Output_Get:Timer Handle Structure)
   n6 = Branch(execute=@n4.else, Condition=Is Timer Active by Handle!(Handle:Timer Handle Structure=Get EmptyDelta_Handle()) -> (ReturnValue:bool))
   EmptyDelta(execute=@n6.else) {is=title K2Node_CallFunction}
  ]}
 SetStaminaProgress {nodes [
   n1 = SetStaminaProgress!() -> (then, CurrentEnergy:double) {is=title K2Node_FunctionEntry}
   > Parent: SetStaminaProgress!(execute, CurrentEnergy:double=@n1) -> (then) {is=title K2Node_CallParentFunction}
   n2 = Get CurrentEnergy() -> (CurrentEnergy:double)
   > SetScalarParameterValue(self=Get FillBarMID() -> (FillBarMID:Material Instance Dynamic Object Reference), ParameterName=Progress, Value=@n2)
   > SetRenderOpacity(
    self=Get SprintIcon()
    InOpacity=Select(Option 0:double=1, Option 1:double=0.33, Index="float <= float"(A=@n2) -> (ReturnValue=~)) -> (ReturnValue:double)
   )
   > SetScalarParameterValue(self=Get GlowBarMID() -> (GlowBarMID:Material Instance Dynamic Object Reference), ParameterName=Progress, Value=@n2)
   n3 = Get CurrentEnergy()
   > SetDelta!(execute, self, CurrentEnergy:double=@n3) -> (then) {is=title K2Node_CallFunction}
   > Set PrevEnergy(PrevEnergy:double=@n3) -> (Output_Get:double)
  ]}
]
variables [
 Anim_DeltaThreshold {type=number, default=0.05}
 bAnimHideWidget {type=boolean}
 bEndingStateActive {type=boolean}
 ConstantEndingStateTime {type=number, default=5}
 ConstantSnowEffectTag {type=GameplayTag, default {tagName=Granted.SnowGrenade.Slow}}
 DeltaBarMID {type=MaterialInstanceDynamic}
 EmptyDelta_Handle {type=TimerHandle}
 FillBarMID {type=MaterialInstanceDynamic}
 GlowBarMID {type=MaterialInstanceDynamic}
 PrevEnergy {type=number}
 SnowTag_Active {type=GameplayTag, default {tagName=Granted.SnowGrenade.Slow}}
]
widgets {
 info {
  inheritedWidgetCount=0
  namedSlotCount=2
  parentClass=/TacticalSprintUI/UI/WBP_TacSprint_Layout.WBP_TacSprint_Layout_C
  rootWidgetClass=null
  widgetCount=5
 }
 widgets [
  SnowSprintOverlay {widgetClassPath=&Overlay, bIsVariable=true, widget=SnowSprintOverlay}
  Snow_OuterGlow {
   widgetClassPath=&Image
   bIsVariable=true
   parent=SnowSprintOverlay
   widget=Snow_OuterGlow
   properties {
    brush {
     bIsDynamicallyLoaded=false
     drawAs=Image
     imageType=FullColor
     mirroring=NoMirror
     resourceName=None
     resourceObject=/FirePetalSnowGrenade/UI/Materials/MI_UI_Snow_Glow
     tiling=NoTile
     imageSize=<32 32>
     margin=<ltrb 0.5 0.5 0.5 0.5>
     outlineSettings {
      bUseBrushTransparency=false
      roundingType=HalfHeightRadius
      width=0
      color {colorUseRule=UseColor_Specified, specifiedColor=<rgba 0 0 0 0>}
      cornerRadii=<0 0 0 0>
     }
     tintColor {colorUseRule=UseColor_Specified, specifiedColor=<rgba 1 1 1 1>}
     uVRegion {bIsValid=false, max=<0 0>, min=<0 0>}
    }
    colorAndOpacity=<rgba 1 1 1 0.4>
   }
   slot {class=OverlaySlot, properties {horizontalAlignment=HAlign_Fill, padding=<ltrb 0 -6 0 -6>}}
  }
  Snow_DeltaBar {
   widgetClassPath=&Image
   bIsVariable=true
   parent=SnowSprintOverlay
   widget=Snow_DeltaBar
   properties {
    visibility=HitTestInvisible
    brush {
     bIsDynamicallyLoaded=false
     drawAs=Image
     imageType=FullColor
     mirroring=NoMirror
     resourceName=None
     resourceObject=/FirePetalSnowGrenade/UI/Materials/MI_UI_Snow_HitPointBar_EnergyDelta
     tiling=NoTile
     imageSize=<32 32>
     margin=<ltrb 0 0 0 0>
     outlineSettings {
      bUseBrushTransparency=false
      roundingType=HalfHeightRadius
      width=0
      color {colorUseRule=UseColor_Specified, specifiedColor=<rgba 0 0 0 0>}
      cornerRadii=<0 0 0 0>
     }
     tintColor {colorUseRule=UseColor_Specified, specifiedColor=<rgba 1 1 1 1>}
     uVRegion {bIsValid=false, max=<0 0>, min=<0 0>}
    }
   }
   slot {class=OverlaySlot, properties {horizontalAlignment=HAlign_Fill}}
  }
  Snow_FillBar {
   widgetClassPath=&Image
   bIsVariable=true
   parent=SnowSprintOverlay
   widget=Snow_FillBar
   properties {brush {
     bIsDynamicallyLoaded=false
     drawAs=Box
     imageType=FullColor
     mirroring=NoMirror
     resourceName=None
     resourceObject=/FirePetalSnowGrenade/UI/Materials/MI_UI_SnowState_Sprint
     tiling=NoTile
     imageSize=<32 32>
     margin=<ltrb 0 0 0 0>
     outlineSettings {
      bUseBrushTransparency=false
      roundingType=HalfHeightRadius
      width=0
      color {colorUseRule=UseColor_Specified, specifiedColor=<rgba 0 0 0 0>}
      cornerRadii=<0 0 0 0>
     }
     tintColor {colorUseRule=UseColor_Specified, specifiedColor=<rgba 1 1 1 1>}
     uVRegion {bIsValid=false, max=<0 0>, min=<0 0>}
    }}
   slot {class=OverlaySlot, properties {horizontalAlignment=HAlign_Fill}}
  }
  SprintIcon {
   widgetClassPath=&Image
   bIsVariable=true
   widget=SprintIcon
   properties {
    brush {
     bIsDynamicallyLoaded=false
     drawAs=Image
     imageType=FullColor
     mirroring=NoMirror
     resourceName=None
     resourceObject=/FirePetalSnowGrenade/UI/Materials/M_UI_TacSprint_Icon
     tiling=NoTile
     imageSize=<32 32>
     margin=<ltrb 0.2 0.2 0.2 0.2>
     outlineSettings {
      bUseBrushTransparency=false
      roundingType=HalfHeightRadius
      width=0
      color {colorUseRule=UseColor_Specified, specifiedColor=<rgba 0 0 0 0>}
      cornerRadii=<0 0 0 0>
     }
     tintColor {colorUseRule=UseColor_Specified, specifiedColor=<rgba 1 1 1 1>}
     uVRegion {bIsValid=false, max=<0 0>, min=<0 0>}
    }
    renderTransform {angle=0, scale=<1 1>, shear=<0 0>, translation=<4 0>}
   }
  }
 ]
}
